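# Unlit translucent sprite material for smoke particles in the Popcorn skydiving cosmetic effect.
# An 8x8 smoke poof texture supplies green times particle color through an emissive adjustment function, and red times particle alpha for opacity.
# A time-driven rotation swirls particles around the actor's Z axis.
# Dynamic parameters add an offset combining an acceleration direction with a move amount along the particle-to-actor vector.
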Material /BRCosmetics/Athena/Effects/Materials/Skydiving/Popcorn/M_PopcornSmoke_Sprite
graph {
 properties {
  bUsedWithNiagaraSprites=true
  bUsedWithParticleSprites=true
  blendMode=BLEND_Translucent
  ditheredLODTransition=true
  pixelDepthOffsetMode=PDOM_Legacy
  shadingModel=MSM_Unlit
  usedShadingModels=Unlit
 }
 nodes [
  n1 = Multiply(A=ScalarParameter() {defaultValue=-0.2, parameterName=SpinSpeed}, B=Time())
  n2 = DynamicParameter() {paramNames [AccelerationOverLife, MoveOutAmount, Param3, Param4]}
  n3 = VectorParameter() {parameterName=AccelerationDirection, defaultValue=<rgba 0 0 20 0>}
  n4 = ActorPositionWS()
  n5 = ParticlePositionWS()
  n6 = Multiply(A=@n2.MoveOutAmount, B=Normalize(VectorInput=Subtract(A=@n4, B=@n5)))
  Add(
   A=RotateAboutAxis(
    NormalizedRotationAxis=Constant3Vector() {constant=<rgba 0 0 1 0>}
    RotationAngle=Frac(None=@n1)
    PivotPoint=ActorPositionWS()
    Position=ParticlePositionWS()
   )
   B=Multiply(A=@n2.AccelerationOverLife, B=Add(A=@n3.RGB, B=@n6))
  )
  n7 = TextureSample() {samplerType=SAMPLERTYPE_LinearColor, texture=/Game/Effects/Fort_Effects/Gameplay/SmokeDustExplosions/Textures/T_SmokePoof_8x8_01}
  n8 = ParticleColor()
  MaterialFunctionCall(Emissive=Multiply(A=@n7.G, B=@n8.RGB)) {materialFunction=&MF_FX_EmissiveAdjustments}
  Multiply(A=@n7.R, B=@n8.A)
  ParticleRelativeTime()
 ]
}
