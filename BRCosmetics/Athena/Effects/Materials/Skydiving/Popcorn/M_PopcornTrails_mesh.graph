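# Lit material for mesh particles, Cascade or Niagara, in the popcorn skydiving trail cosmetic.
# A sharpened vertex-color red mask blends outer and inner normals, roughness and two colors, with noise varying the outer color and Fresnel shading the result.
# A position offset spins particles around the actor's Z axis over time, adding a directional push and a move along the particle-to-actor vector, both scaled through dynamic parameters.
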Material /BRCosmetics/Athena/Effects/Materials/Skydiving/Popcorn/M_PopcornTrails_mesh
graph {
 properties {
  bUsedWithMeshParticles=true
  bUsedWithNiagaraMeshParticles=true
  ditheredLODTransition=true
  pixelDepthOffsetMode=PDOM_Legacy
  usedShadingModels=Default Lit
 }
 nodes [
  n1 = Multiply(A=ScalarParameter() {defaultValue=-0.2, parameterName=SpinSpeed}, B=Time())
  n2 = DynamicParameter() {paramNames [AccelerationOverLife, MoveOutAmount, Param3, Param4]}
  n3 = VectorParameter() {parameterName=AccelerationDirection, defaultValue=<rgba 0 0 200 0>}
  n4 = ActorPositionWS()
  n5 = ParticlePositionWS()
  n6 = Multiply(A=@n2.MoveOutAmount, B=Normalize(VectorInput=Subtract(A=@n4, B=@n5)))
  Add(
   A=RotateAboutAxis(
    NormalizedRotationAxis=Constant3Vector() {constant=<rgba 0 0 1 0>}
    RotationAngle=Frac(None=@n1)
    PivotPoint=ActorPositionWS()
    Position=ParticlePositionWS()
   )
   B=Multiply(A=@n2.AccelerationOverLife, B=Add(A=@n3.RGB, B=@n6))
  )
  n7 = TextureCoordinate()
  n8 = TextureSampleParameter2D(UVs=@n7) {parameterName=NormalOutside, samplerType=SAMPLERTYPE_Normal, texture=/Game/Global/Textures/Default/Tileable_Patterns/2ndInfectedMapNormal}
  n9 = VectorParameter() {parameterName=NormalTiles, defaultValue=<rgba 0.8 1.3 0.3 0.3>}
  n10 = TextureSampleParameter2D(UVs=Multiply(A=AppendVector(A=@n9.B, B=@n9.B), B=TextureCoordinate() {uTiling=0.3, vTiling=0.3})) {parameterName=NormalInside, samplerType=SAMPLERTYPE_Normal, texture=/Game/VisualThreatManager/StormVisuals/Textures/TilingNoise06_N}
  n11 = VertexColor()
  n12 = ScalarParameter() {defaultValue=2, parameterName=VertColorMult}
  n13 = Saturate(None=Power(Base=Multiply(A=@n11.R, B=@n12) {constB=2}, Exp=ScalarParameter() {defaultValue=2, parameterName=VerColorExp}))
  LinearInterpolate(
   A=MaterialFunctionCall(Normal=@n8.RGB, Flatness=ScalarParameter() {defaultValue=0.6, parameterName=FlatnessOutside}) {materialFunction=&FlattenNormal}.Result
   B=MaterialFunctionCall(Normal=@n10.RGB, Flatness=ScalarParameter() {parameterName=FlatnessInside}) {materialFunction=&FlattenNormal}.Result
   Alpha=@n13
  )
  LinearInterpolate(A=ScalarParameter() {defaultValue=0.8, parameterName=Rough}, B=ScalarParameter() {defaultValue=0.4, parameterName=RoughInner}, Alpha=@n13)
  n14 = ScalarParameter() {defaultValue=3, parameterName=FresnelExp}
  n15 = VectorParameter() {parameterName=Color1, defaultValue=<rgba 0.9167 0.8449 0.5395 0>}
  n16 = Multiply(A=@n15.RGB, B=@n15.RGB)
  n17 = TextureSampleParameter2D(UVs=Multiply(A=TextureCoordinate(), B=ScalarParameter() {defaultValue=0.6, parameterName=DiffuseTile})) {parameterName=DiffuseNoise, samplerType=SAMPLERTYPE_LinearColor, texture=/Game/Weapons/WeaponSkins/Textures/T_Swirly_Noise_highContrast1}
  n18 = Multiply(A=@n17.R, B=ScalarParameter() {defaultValue=1, parameterName=DiffuseAmnt})
  n19 = VectorParameter() {parameterName=Color2, defaultValue=<rgba 0.4427 0.2086 0.0784 0>}
  n20 = OneMinus(None=Multiply(A=@n11.R, B=ScalarParameter() {parameterName=VertColorOverlayAmnt}))
  Multiply(
   A=LinearInterpolate(Alpha=MaterialFunctionCall(Exponent=@n14) {materialFunction=&MF_Fresnel_FX}.Fresnel) {constA=0.2}
   B=LinearInterpolate(A=LinearInterpolate(A=@n15.RGB, B=@n16, Alpha=@n18) {constA=0.6}, B=Multiply(A=@n19.RGB, B=@n20), Alpha=@n13)
  )
  Multiply(A=@n7, B=ComponentMask(None=@n9.RGB) {g=true, r=true})
  Multiply(A=@n18)
 ]
}
