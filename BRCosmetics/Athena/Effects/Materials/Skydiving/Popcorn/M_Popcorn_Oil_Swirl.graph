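# Translucent, unlit, two-sided particle material for a swirling oil effect in the Popcorn skydiving cosmetic.
# Tiled UVs get a per-particle random U offset, panning-noise distortion and scrolling, then sample a sea-foam normal and mask.
# The flattened normal drives a cubemap reflection that is tinted and added to a base color, faking glossy lighting.
# Opacity multiplies the foam mask by UV edge fades, then uses MF_ClipAlpha with particle alpha to clip it over time.
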Material /BRCosmetics/Athena/Effects/Materials/Skydiving/Popcorn/M_Popcorn_Oil_Swirl
graph {
 properties {
  bUsedWithMeshParticles=true
  bUsedWithNiagaraMeshParticles=true
  bUsedWithNiagaraSprites=true
  bUsedWithParticleSprites=true
  blendMode=BLEND_Translucent
  pixelDepthOffsetMode=PDOM_Legacy
  shadingModel=MSM_Unlit
  twoSided=true
  usedShadingModels=Unlit
 }
 nodes [
  n1 = TextureCoordinate()
  n2 = Multiply(A=@n1, B=ComponentMask(None=VectorParameter() {parameterName=Tile, defaultValue=<rgba 3 3 0 0>}.RGB) {g=true, r=true})
  n3 = Add(
   A=DynamicParameter() {defaultValue=<rgba 0 1 1 1>, paramNames [rand 0-1, Param2, Param3, Param4]}.rand 0-1
   B=ComponentMask(None=@n2) {r=true}
  )
  n4 = ComponentMask(None=@n2) {g=true}
  n5 = ComponentMask(None=TextureSample(UVs=Panner(Coordinate=@n1) {speedX=-0.1}) {texture=/Game/Global/Textures/Default/Tileable_Patterns/T_Tilling_Distortion_Map_01}.G) {g=true, r=true}
  n6 = ScalarParameter() {defaultValue=0.1, parameterName=DistAmnt}
  n7 = Panner(
   Coordinate=Add(A=AppendVector(A=@n3, B=@n4), B=Multiply(A=@n5, B=@n6))
   Speed=ComponentMask(None=VectorParameter() {parameterName=PanSpeed, defaultValue=<rgba 0.3 0 0 0>}.RGB) {g=true, r=true}
  ) {speedX=0.3}
  n8 = TextureSampleParameter2D(UVs=@n7) {
   parameterName=normal
   samplerSource=SSM_Wrap_WorldGroupSettings
   samplerType=SAMPLERTYPE_Normal
   texture=/Game/Packages/Fortress_Transylvania/Haunted_Forest/Swamps/Textures/T_sea_foam_03_n
  }
  n9 = ScalarParameter() {parameterName=NormalFlatness}
  n10 = Transform(None=MaterialFunctionCall(Normal=@n8.RGB, Flatness=@n9) {materialFunction=&FlattenNormal}.Result)
  Add(A=CameraVectorWS(), B=@n10)
  n11 = Multiply(
   A=VectorParameter() {parameterName=Color2, defaultValue=<rgba 0.2448 0.1234 0 0>}.RGB
   B=TextureSampleParameterCube!(UVs=ReflectionVectorWS(CustomWorldNormal=@n10), Apply View MipBias) {parameterName=Cube, samplerType=SAMPLERTYPE_LinearColor, texture=/Game/Weapons/WeaponSkins/Textures/AthenaCubemap2}.B
  )
  n12 = ScalarParameter() {defaultValue=3, parameterName=CubeBrightness}
  MaterialFunctionCall(Emissive=Add(A=VectorParameter() {parameterName=Color1, defaultValue=<rgba 0.2448 0.1234 0 0>}.RGB, B=Multiply(A=@n11, B=@n12))) {materialFunction=&MF_FX_EmissiveAdjustments}
  Multiply(A=@n8.RGB, B=AppendVector(A=AppendVector(A=@n9, B=@n9), B=Constant() {r=1}))
  n13 = ComponentMask(None=TextureCoordinate()) {g=true}
  n14 = ScalarParameter() {defaultValue=8, parameterName=EdgeMult}
  n15 = ScalarParameter() {defaultValue=3, parameterName=EdgeExp}
  n16 = Saturate(None=Power(Base=Multiply(A=@n13, B=@n14), Exp=@n15))
  n17 = OneMinus(None=@n13)
  n18 = Saturate(None=Power(Base=Multiply(A=@n17, B=@n14), Exp=@n15))
  n19 = ComponentMask(None=TextureCoordinate()) {r=true}
  n20 = ComponentMask(None=@n1) {r=true}
  n21 = Power(Base=Multiply(A=OneMinus(None=@n20)) {constB=10}) {constExponent=4}
  n22 = Multiply(A=Saturate(None=Multiply(A=@n16, B=@n18)), B=Multiply(A=@n19, B=Saturate(None=@n21)))
  n23 = TextureSampleParameter2D(UVs=@n7) {
   parameterName=Mask
   samplerSource=SSM_Wrap_WorldGroupSettings
   texture=/Game/Packages/Fortress_Transylvania/Haunted_Forest/Swamps/Textures/T_sea_foam_03
  }
  n24 = Multiply(
   A=ChannelMaskParameter(None=@n23.RGB) {maskChannel=Blue, parameterName=MaskChannel, defaultValue=<rgba 0 0 1 0>}.R
   B=ScalarParameter() {defaultValue=1.5, parameterName=OpacPreMult}
  )
  Multiply(
   A=MaterialFunctionCall!(
    AlphaTexture=Multiply(A=@n22, B=@n24)
    ClipRange=ScalarParameter() {defaultValue=0.08, parameterName=ClipRng}
    AlphaOverTime=ParticleColor().A
   ) {materialFunction=/Game/Effects/Fort_Effects/Materials/_Functions/MF_ClipAlpha}.Result
   B=ScalarParameter() {defaultValue=1, parameterName=OpacPostMult}
  )
  OneMinus(None=@n19)
 ]
}
